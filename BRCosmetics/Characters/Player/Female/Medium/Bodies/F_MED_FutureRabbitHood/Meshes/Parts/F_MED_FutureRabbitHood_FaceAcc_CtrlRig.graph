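# Corrective Control Rig for the FutureRabbitHood female character's face accessory, adjusting hood, ear and hair bones.
# A spherical pose reader on upperarm_r relative to spine_05, with debug draw on, drives right-side hood and hair offsets.
# Both alt-skin ear correctives get constant full-weight offsets.
# The `Backpack cape` bool switches on weighted parent constraints and offsets that reposition the hood ears; the DeltaRotation library function is unused in the main graph.
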ControlRigBlueprint /BRCosmetics/Characters/Player/Female/Medium/Bodies/F_MED_FutureRabbitHood/Meshes/Parts/F_MED_FutureRabbitHood_FaceAcc_CtrlRig
graphs [
 New Function_ContainedGraph {nodes [
   n1 = RigVMFunctionEntryNode()
   RigVMFunction_MathBoolToInteger = &RigVMFunction_MathBoolToInteger(Value=@n1.onlyTwist)
   RigVMFunction_MathQuaternionSwingTwist = &RigVMFunction_MathQuaternionSwingTwist(
    twistAxis=@n1.twistAxis
    Input=&RigUnit_GetRelativeTransformForItem(bChildInitial=False, bParentInitial=False, child=@n1.child, parent=@n1.parent).RelativeTransform.Rotation
   )
   n2 = RigVMDispatchNode(
    Values=((),())
    Index=@RigVMFunction_MathBoolToInteger.Result
    Values.0=@RigVMFunction_MathQuaternionSwingTwist.Swing
    Values.1=@RigVMFunction_MathQuaternionSwingTwist.Twist
   ) {template="DISPATCH_RigVMDispatch_SelectInt32(in Index,in Values,out Result)"}
   RigVMFunction_MathQuaternionToEuler = &RigVMFunction_MathQuaternionToEuler(Value=@n2.Result, RotationOrder=@n1.Rotation Order)
   > DISPATCH_RigVMDispatch_Print = RigVMDispatchNode(ScreenColor=(R=1,G=1,B=1,A=1), ScreenDuration=1.5, Enabled=@n1.debugRotation, Value=@RigVMFunction_MathQuaternionToEuler.Result) {template="DISPATCH_RigVMDispatch_Print(in Prefix,in Value,in Enabled,in ScreenDuration,in ScreenColor)"}
   Cast = RigVMDispatchNode(Value=@n1.triggerAxis) {template="DISPATCH_RigVMDispatch_CastEnumToInt(in Value,out Result)"}
   DISPATCH_RigVMDispatch_SelectInt32 = RigVMDispatchNode(
    Values=(0,0,0)
    Index=@Cast.Result
    Values.0=@RigVMFunction_MathQuaternionToEuler.Result.X
    Values.1=@RigVMFunction_MathQuaternionToEuler.Result.Y
    Values.2=@RigVMFunction_MathQuaternionToEuler.Result.Z
   ) {template="DISPATCH_RigVMDispatch_SelectInt32(in Index,in Values,out Result)"}
   Remap = &RigVMFunction_MathDoubleRemap(
    bClamp=False
    Value=@DISPATCH_RigVMDispatch_SelectInt32.Result
    TargetMinimum=@n1.outputMinMax.X
    TargetMaximum=@n1.outputMinMax.Y
    SourceMinimum=@n1.rotationMinMax.X
    SourceMaximum=@n1.rotationMinMax.Y
   )
   Absolute = &RigVMFunction_MathDoubleAbs(Value=@Remap.Result)
   Multiply = &RigVMFunction_MathDoubleMul(
    A=&RigVMFunction_MathDoubleClamp(Value=RigVMDispatchNode(True=@Absolute.Result, Condition=@n1.absoluteValue, False=@Remap.Result) {template="DISPATCH_RigVMDispatch_If(in Condition,in True,in False,out Result)"}.Result).Result
    B=@n1.premult
   )
   > RigVMFunctionReturnNode(outputQuat=@n2.Result, outputWeight=@Multiply.Result)
  ]}
 RigVMFunctionLibrary {nodes [DeltaRotation v4 = RigVMCollapseNode(
    ExecuteContext=()
    Rotation Order=ZYX
    child=(Type=Bone,Name="None")
    outputMinMax=(X=0,Y=1)
    parent=(Type=Bone,Name="None")
    premult=1
    rotationMinMax=(X=0,Y=0)
    twistAxis=(X=1,Y=0,Z=0)
   )]}
 RigVMFunction_Sequence_2_ContainedGraph {nodes [
   n1 = RigVMFunctionEntryNode()
   n2 = &RigVMFunction_Sequence(ExecuteContext=@n1.ExecuteContext)
   n3 = &RigVMFunction_Sequence(ExecuteContext=@n2.B)
   RigVMFunctionReturnNode(A=@n2.A, B=@n3.A, C=@n3.B)
  ]}
 RigVMFunction_Sequence_2_ContainedGraph {nodes [
   n1 = RigVMFunctionEntryNode()
   RigVMFunction_Sequence = &RigVMFunction_Sequence(ExecuteContext=@n1.ExecuteContext)
   n2 = &RigVMFunction_Sequence(ExecuteContext=@RigVMFunction_Sequence.B)
   RigVMFunctionReturnNode(A=@RigVMFunction_Sequence.A, B=@n2.A, C=@n2.B)
  ]}
 RigVMModel {nodes [
   RigUnit_BeginExecution = &RigUnit_BeginExecution()
   RigVMFunction_Sequence_1 = RigVMAggregateNode(ExecuteContext=@RigUnit_BeginExecution.ExecutePin)
   RigVMFunction_Sequence = RigVMAggregateNode(ExecuteContext=@RigVMFunction_Sequence_1.C)
   n1 = &RigUnit_SphericalPoseReader(
    ActiveRegionScaleFactors=(PositiveWidth=1,NegativeWidth=1,PositiveHeight=1,NegativeHeight=1)
    ActiveRegionSize=0.1
    Debug=(bDrawDebug=True,bDraw2D=False,bDrawLocalAxes=False,DebugScale=25,DebugSegments=20,DebugThickness=0.25)
    DriverAxis=(X=1,Y=0,Z=0)
    DriverItem=(Type=Bone,Name="upperarm_r")
    FalloffRegionScaleFactors=(PositiveWidth=1,NegativeWidth=1,PositiveHeight=1,NegativeHeight=1)
    FalloffSize=0.2
    FlipHeightScaling=False
    FlipWidthScaling=False
    OptionalParentItem=(Type=Bone,Name="spine_05")
    RotationOffset=(X=70,Y=0,Z=60)
    ExecutePin=@RigVMFunction_Sequence.A
   )
   > &RigUnit_ModifyTransforms(
    ItemToModify=((Item=(Type=Bone,Name="dyn_hood_corr_r_base"),Transform=(Rotation=(X=0,Y=0,Z=0,W=-1),Translation=(X=5.9054,Y=-2.8938,Z=19.9503),Scale3D=(X=1,Y=1,Z=1))))
    WeightMinimum=0
    Weight=@n1.OutputParam
   )
   > &RigUnit_ModifyTransforms(
    ItemToModify=((Item=(Type=Bone,Name="dyn_hair_r_base"),Transform=(Rotation=(X=0.021,Y=0.0903,Z=0.2228,W=-0.9705),Translation=(X=0,Y=0,Z=0),Scale3D=(X=1,Y=1,Z=1))))
    WeightMinimum=0
    Weight=@n1.OutputParam
   )
   &RigUnit_ModifyTransforms(
    ItemToModify=((Item=(Type=Bone,Name="dyn_altskin_corr_ear_l_06"),Transform=(Rotation=(X=-0.1202,Y=-0.3851,Z=0.0976,W=-0.9098),Translation=(X=-2.9108,Y=-0.7786,Z=-21.8194),Scale3D=(X=1,Y=1,Z=1))))
    Weight=1
    WeightMinimum=0
    ExecutePin=@RigVMFunction_Sequence.B
   )
   > &RigUnit_ModifyTransforms(
    ItemToModify=((Item=(Type=Bone,Name="dyn_altskin_corr_ear_r_06"),Transform=(Rotation=(X=-0.0338,Y=-0.3562,Z=0.0541,W=-0.9322),Translation=(X=-6.3341,Y=-1.1258,Z=-17.9706),Scale3D=(X=1,Y=1,Z=1))))
    Weight=1
    WeightMinimum=0
   )
   n2 = RigVMDispatchNode(False=0, True=1, Condition=RigVMVariableNode() {variable=Backpack cape}.Value) {template="DISPATCH_RigVMDispatch_If(in Condition,in True,in False,out Result)"}
   &RigUnit_ParentConstraint(
    AdvancedSettings=(InterpolationType=Average,RotationOrderForFilter=XZY)
    Child=(Type=Bone,Name="dyn_hood_ear_l_03")
    Parents="((Item=(Type=Bone,Name=\"neck_01\"),Weight=0.0086),(Item=(Type=Bone,Name=\"neck_02\"),Weight=0.6069),(Item=(Type=Bone,Name=\"head\"),Weight=0.0003),(Item=(Type=Bone,Name=\"clavicle_l\"),Weight=0.1827),(Item=(Type=Bone,Name=\"clavicle_r\"),Weight=0.0032),(Item=(Type= ...(truncated)"
    Weight=@n2.Result
    ExecutePin=@RigVMFunction_Sequence.C
   )
   > &RigUnit_ParentConstraint(
    AdvancedSettings=(InterpolationType=Average,RotationOrderForFilter=XZY)
    Child=(Type=Bone,Name="dyn_hood_ear_l_04")
    Parents="((Item=(Type=Bone,Name=\"neck_01\"),Weight=0.0058),(Item=(Type=Bone,Name=\"neck_02\"),Weight=0.3627),(Item=(Type=Bone,Name=\"dyn_hood_ear_l_05\"),Weight=0.0151),(Item=(Type=Bone,Name=\"clavicle_l\"),Weight=0.1633),(Item=(Type=Bone,Name=\"clavicle_r\"),Weight=0.0031) ...(truncated)"
    Weight=@n2.Result
   )
   > &RigUnit_ParentConstraint(
    AdvancedSettings=(InterpolationType=Average,RotationOrderForFilter=XZY)
    Child=(Type=Bone,Name="dyn_hood_ear_l_05")
    Parents="((Item=(Type=Bone,Name=\"neck_01\"),Weight=0.001),(Item=(Type=Bone,Name=\"neck_02\"),Weight=0.1221),(Item=(Type=Bone,Name=\"dyn_hood_ear_l_05\"),Weight=0.0425),(Item=(Type=Bone,Name=\"clavicle_l\"),Weight=0.0355),(Item=(Type=Bone,Name=\"clavicle_r\"),Weight=0.0007), ...(truncated)"
    Weight=@n2.Result
   )
   > &RigUnit_ModifyTransforms(
    ItemToModify=((Item=(Type=Bone,Name="dyn_hood_ear_l_03"),Transform=(Rotation=(X=0.1205,Y=-0.3332,Z=-0.042,W=-0.9342),Translation=(X=-1.8315,Y=-4.6866,Z=-1.1925),Scale3D=(X=1,Y=1,Z=1))))
    WeightMinimum=0
    Weight=@n2.Result
   )
   > &RigUnit_ModifyTransforms(
    ItemToModify=((Item=(Type=Bone,Name="dyn_hood_ear_l_04"),Transform=(Rotation=(X=0.0226,Y=-0.1655,Z=-0.0021,W=-0.986),Translation=(X=-1.2493,Y=3.3367,Z=-3.6528),Scale3D=(X=1,Y=1,Z=1))))
    WeightMinimum=0
    Weight=@n2.Result
   )
   > &RigUnit_ParentConstraint(
    AdvancedSettings=(InterpolationType=Average,RotationOrderForFilter=XZY)
    Child=(Type=Bone,Name="dyn_hood_ear_r_03")
    Parents="((Item=(Type=Bone,Name=\"neck_02\"),Weight=0.7912),(Item=(Type=Bone,Name=\"dyn_hood_ear_r_05\"),Weight=0.0903),(Item=(Type=Bone,Name=\"upperarm_r\"),Weight=0),(Item=(Type=Bone,Name=\"clavicle_l\"),Weight=0.0002),(Item=(Type=Bone,Name=\"clavicle_r\"),Weight=0.0139),( ...(truncated)"
    Weight=@n2.Result
   )
   > &RigUnit_ParentConstraint(
    AdvancedSettings=(InterpolationType=Average,RotationOrderForFilter=XZY)
    Child=(Type=Bone,Name="dyn_hood_ear_r_04")
    Parents="((Item=(Type=Bone,Name=\"neck_02\"),Weight=0.4265),(Item=(Type=Bone,Name=\"dyn_hood_ear_r_05\"),Weight=0.1146),(Item=(Type=Bone,Name=\"upperarm_r\"),Weight=0.0003),(Item=(Type=Bone,Name=\"clavicle_l\"),Weight=0.0003),(Item=(Type=Bone,Name=\"clavicle_r\"),Weight=0.01 ...(truncated)"
    Weight=@n2.Result
   )
   > &RigUnit_ParentConstraint(
    AdvancedSettings=(InterpolationType=Average,RotationOrderForFilter=XZY)
    Child=(Type=Bone,Name="dyn_hood_ear_r_05")
    Parents="((Item=(Type=Bone,Name=\"neck_02\"),Weight=0.1009),(Item=(Type=Bone,Name=\"dyn_hood_ear_r_05\"),Weight=0.0588),(Item=(Type=Bone,Name=\"upperarm_r\"),Weight=0.0001),(Item=(Type=Bone,Name=\"clavicle_l\"),Weight=0.0001),(Item=(Type=Bone,Name=\"clavicle_r\"),Weight=0.00 ...(truncated)"
    Weight=@n2.Result
   )
   > &RigUnit_ModifyTransforms(
    ItemToModify=((Item=(Type=Bone,Name="dyn_hood_ear_r_03"),Transform=(Rotation=(X=-0.0171,Y=0.4449,Z=0.0501,W=0.894),Translation=(X=-0.6138,Y=5.8207,Z=0.0817),Scale3D=(X=1,Y=1,Z=1))))
    WeightMinimum=0
    Weight=@n2.Result
   )
   > &RigUnit_ModifyTransforms(
    ItemToModify=((Item=(Type=Bone,Name="dyn_hood_ear_r_04"),Transform=(Rotation=(X=-0.0268,Y=-0.0857,Z=0.0357,W=-0.9953),Translation=(X=-0.4362,Y=-1.3694,Z=-2.3797),Scale3D=(X=1,Y=1,Z=1))))
    WeightMinimum=0
    Weight=@n2.Result
   )
  ]}
]
hierarchy [
 root {loc=<0 0 0>, rot=<0 0 0 1>}
 pelvis {parent=root, loc=<0 -0.5301 98.0571>, rot=<0.0127 0.707 -0.0127 -0.707>}
 spine_01 {parent=pelvis, loc=<6.1208 -0.9328 0>, rot=<0 0 0.034 -0.9994>}
 spine_02 {parent=spine_01, loc=<6.5997 -0.2095 0>, rot=<0 0 0.0238 -0.9997>}
 spine_03 {parent=spine_02, loc=<6.4914 -0.1237 0>, rot=<0 0 0.0149 -0.9999>}
 spine_04 {parent=spine_03, loc=<6.9598 0.0794 0>, rot=<0 0 -0.0132 -0.9999>}
 spine_05 {parent=spine_04, loc=<6.9987 0 0>, rot=<0 0 -0.0015 -1>}
 neck_01 {parent=spine_05, loc=<11.3138 0.1529 0>, rot=<0 0 -0.2003 -0.9797>}
 neck_02 {parent=neck_01, loc=<4.5627 0 0>, rot=<0 0 0.0504 -0.9987>}
 head {parent=neck_02, loc=<5.3914 0 0>, rot=<0 0 0.115 -0.9934>}
 dyn_hair_front_1_l_base {parent=head, loc=<5.7377 9.2216 -1.5451>, rot=<0.0041 -1 0 0>}
 dyn_hair_front_1_l_01 {parent=dyn_hair_front_1_l_base, loc=<0 0 0>, rot=<0 0 0 1>}
 dyn_hair_front_base {parent=head, loc=<4.4328 9.0201 1.5213>, rot=<0.0041 -1 0 0>}
 dyn_hair_front_01 {parent=dyn_hair_front_base, loc=<0 0 0>, rot=<0 0 0 1>}
 dyn_hair_l_base {parent=head, loc=<2.8121 1.1378 -6.9281>, rot=<-0.0344 -0.9982 0.0253 0.0426>}
 dyn_hair_l_01 {parent=dyn_hair_l_base, loc=<0 0 0>, rot=<0 0 0 1>}
 dyn_hair_l_02 {parent=dyn_hair_l_01, loc=<7.5832 0 0>, rot=<0 0 0 1>}
 dyn_hair_r_base {parent=head, loc=<3.1095 1.0697 7.4523>, rot=<-0.032 -0.9986 -0.0136 -0.0399>}
 dyn_hair_r_01 {parent=dyn_hair_r_base, loc=<0 0 0>, rot=<0 0 0 1>}
 dyn_hair_r_02 {parent=dyn_hair_r_01, loc=<7.0192 0 0>, rot=<0 0 0 1>}
 dyn_hood_ear_r_base {parent=head, loc=<13.5446 0.0797 5.6639>, rot=<0.481 -0.3697 0.7401 -0.29>}
 dyn_hood_ear_r_01 {parent=dyn_hood_ear_r_base, loc=<0 0 0>, rot=<-0.0016 0.03 -0.0074 -0.9995>}
 dyn_hood_ear_r_02 {parent=dyn_hood_ear_r_01, loc=<7.3644 0 0.0004>, rot=<0 -0.1879 0 -0.9822>}
 dyn_hood_ear_r_03 {parent=dyn_hood_ear_r_02, loc=<10.9998 0 0.0002>, rot=<0 -0.0171 0 -0.9999>}
 dyn_hood_ear_r_04 {parent=dyn_hood_ear_r_03, loc=<11.0002 0 -0.0002>, rot=<0 0.0943 0 -0.9955>}
 dyn_hood_ear_r_05 {parent=dyn_hood_ear_r_04, loc=<11.0002 0 -0.0003>, rot=<0 0 0 1>}
 dyn_hood_ear_l_base {parent=head, loc=<13.5448 0.0797 -5.6639>, rot=<0.29 -0.7401 0.3697 -0.481>}
 dyn_hood_ear_l_01 {parent=dyn_hood_ear_l_base, loc=<0 0 0>, rot=<0.0016 0.03 0.0074 -0.9995>}
 dyn_hood_ear_l_02 {parent=dyn_hood_ear_l_01, loc=<7.3645 0 0>, rot=<0 -0.1879 0 -0.9822>}
 dyn_hood_ear_l_03 {parent=dyn_hood_ear_l_02, loc=<11 0 0>, rot=<0 -0.0171 0 -0.9999>}
 dyn_hood_ear_l_04 {parent=dyn_hood_ear_l_03, loc=<11 0 0>, rot=<0 0.0943 0 -0.9955>}
 dyn_hood_ear_l_05 {parent=dyn_hood_ear_l_04, loc=<11 0 0>, rot=<0 0 0 1>}
 dyn_hair_bk_l_base {parent=head, loc=<2.7019 -8.9044 -2.5699>, rot=<0.0041 -1 0 0>}
 dyn_hair_bk_l_01 {parent=dyn_hair_bk_l_base, loc=<0 0 0>, rot=<0 0 0 1>}
 dyn_hair_bk_r_base {parent=head, loc=<2.7059 -6.6516 6.1668>, rot=<0.0041 -1 0 0>}
 dyn_hair_bk_r_01 {parent=dyn_hair_bk_r_base, loc=<0 0 0>, rot=<0 0 0 1>}
 dyn_hair_right_1_base {parent=head, loc=<3.6123 -2.2971 8.1325>, rot=<0.0041 -1 0 0>}
 dyn_hair_right_1_01 {parent=dyn_hair_right_1_base, loc=<0 0 0>, rot=<0 0 0 1>}
 dyn_hair_right_base {parent=head, loc=<3.4695 4.0073 8.763>, rot=<0.0041 -1 0 0>}
 dyn_hair_right_01 {parent=dyn_hair_right_base, loc=<0 0 0>, rot=<0 0 0 1>}
 dyn_hair_left_1_base {parent=head, loc=<3.7284 -4.105 -8.4456>, rot=<0.0041 -1 0 0>}
 dyn_hair_left_1_01 {parent=dyn_hair_left_1_base, loc=<0 0 0>, rot=<0 0 0 1>}
 dyn_hair_left_base {parent=head, loc=<3.1534 4.6853 -8.1798>, rot=<0.0041 -1 0 0>}
 dyn_hair_left_01 {parent=dyn_hair_left_base, loc=<0 0 0>, rot=<0 0 0 1>}
 corr_hood_1 {parent=neck_01, loc=<13.112 0 0>, rot=<0 0 0 -1>}
 dyn_hood_corr_l_base {parent=spine_05, loc=<14.5252 2.2363 -11.5894>, rot=<0.0401 -0.9992 0 0>}
 dyn_hood_corr_l_01 {parent=dyn_hood_corr_l_base, loc=<0 0 0>, rot=<0 0 0 1>}
 dyn_hood_corr_r_base {parent=spine_05, loc=<13.8473 2.4617 11.2065>, rot=<0.0401 -0.9992 0 0>}
 dyn_hood_corr_r_01 {parent=dyn_hood_corr_r_base, loc=<0 0 0>, rot=<0 0 0 1>}
 clavicle_l {parent=spine_05, loc=<5.4831 -0.1053 -3.2946>, rot=<0.0465 -0.7289 -0.009 -0.6829>}
 upperarm_l {parent=clavicle_l, loc=<9.8658 -1.2136 -0.4105>, rot=<0.0164 -0.3245 -0.0269 -0.9454>}
 lowerarm_l {parent=upperarm_l, loc=<26.465 -0.8345 0.259>, rot=<0.0036 0.0161 -0.2298 -0.9731>}
 hand_l {parent=lowerarm_l, loc=<22.7355 -0.3203 -0.6105>, rot=<-0.6784 -0.0223 0.0735 -0.7307>}
 hand_attach_l {parent=hand_l, loc=<8.5 -3 0>, rot=<0.7071 0.7071 0 0>}
 weapon_l {parent=hand_l, loc=<-0.1974 -3.4111 0.3811>, rot=<0 0 0 1>}
 thumb_01_l {parent=hand_l, loc=<1.3118 -0.3191 -2.459>, rot=<0.6546 -0.4042 0.0664 -0.6354>}
 thumb_02_l {parent=thumb_01_l, loc=<5.9336 -0.0448 0.0401>, rot=<-0.0037 0.0213 0.298 -0.9543>}
 thumb_03_l {parent=thumb_02_l, loc=<2.2223 0.3682 -0.0907>, rot=<0.0118 -0.013 0.0066 -0.9998>}
 index_metacarpal_l {parent=hand_l, loc=<1.6535 0.0183 -2.96>, rot=<0.0158 -0.0124 -0.0173 -0.9996>}
 index_01_l {parent=index_metacarpal_l, loc=<8.3008 0.0357 -0.0524>, rot=<-0.0044 0.0485 0.0924 -0.9945>}
 index_02_l {parent=index_01_l, loc=<3.3587 -0.3004 0.0452>, rot=<-0.002 0.0027 0.3225 -0.9466>}
 index_03_l {parent=index_02_l, loc=<2.5578 -0.123 -0.0252>, rot=<-0.002 0.0067 0.2558 -0.9667>}
 middle_metacarpal_l {parent=hand_l, loc=<2.1856 -0.1725 -0.5984>, rot=<0.006 0.0168 -0.0329 -0.9993>}
 middle_01_l {parent=middle_metacarpal_l, loc=<7.3275 -0.2447 0.0129>, rot=<-0.0208 0.0042 0.0858 -0.9961>}
 middle_02_l {parent=middle_01_l, loc=<4.0123 0.1401 0.1175>, rot=<-0.0218 0.0258 0.3512 -0.9357>}
 middle_03_l {parent=middle_02_l, loc=<2.6077 0.0534 0.1373>, rot=<0.0082 0.0188 0.1708 -0.9851>}
 ring_metacarpal_l {parent=hand_l, loc=<2.2074 -0.3625 1.3638>, rot=<-0.0284 0.0722 -0.0458 -0.9959>}
 ring_01_l {parent=ring_metacarpal_l, loc=<7.0295 -0.6197 -0.1446>, rot=<-0.0489 -0.0252 0.1892 -0.9804>}
 ring_02_l {parent=ring_01_l, loc=<3.6174 0.4647 -0.1027>, rot=<0 0 0.2806 -0.9598>}
 ring_03_l {parent=ring_02_l, loc=<2.4152 0.114 0.0199>, rot=<0 0 0.1189 -0.9929>}
 pinky_metacarpal_l {parent=hand_l, loc=<1.7616 -0.9753 3.3814>, rot=<-0.0397 0.0814 -0.069 -0.9935>}
 pinky_01_l {parent=pinky_metacarpal_l, loc=<7.1709 -0.4831 -0.1094>, rot=<-0.0654 -0.0035 0.2242 -0.9723>}
 pinky_02_l {parent=pinky_01_l, loc=<2.8274 0.0893 -0.1043>, rot=<0.0037 0.0011 0.2854 -0.9584>}
 pinky_03_l {parent=pinky_02_l, loc=<2.2046 0.0307 -0.0078>, rot=<-0.0002 0.0098 0.1445 -0.9895>}
 deform_wrist_dn_base_l {parent=hand_l, loc=<0 -1.5326 0>, rot=<0 0 0 -1>}
 deform_wrist_dn_l {parent=deform_wrist_dn_base_l, loc=<0 0 0>, rot=<0 0 0 1>}
 deform_wrist_up_base_l {parent=hand_l, loc=<0 2.3009 0>, rot=<0 0 0 -1>}
 deform_wrist_up_l {parent=deform_wrist_up_base_l, loc=<0 0 0>, rot=<0 0 0 1>}
 lowerarm_twist_02_l {parent=lowerarm_l, loc=<10.6707 0 0>, rot=<0 0 0 1>}
 lowerarm_twist_01_l {parent=lowerarm_l, loc=<18.316 0 0>, rot=<0 0 0 1>}
 deform_elbow_in_base_l {parent=lowerarm_l, loc=<0.1834 3.6062 -0.2572>, rot=<0 0 0 -1>}
 deform_elbow_in_l {parent=deform_elbow_in_base_l, loc=<0 0 0>, rot=<0 0 0 1>}
 deform_elbow_out_base_l {parent=lowerarm_l, loc=<0 -4 0>, rot=<0 0 0 1>}
 deform_elbow_out_l {parent=deform_elbow_out_base_l, loc=<0 0 0>, rot=<0 0 0 1>}
 upperarm_twist_01_l {parent=upperarm_l, loc=<7.2135 -0.1134 0>, rot=<0 0 0 1>}
 deform_upperarm_bk_base_l {parent=upperarm_twist_01_l, loc=<-6.4053 -4.3746 0.314>, rot=<0 0 0 -1>}
 deform_upperarm_bk_l {parent=deform_upperarm_bk_base_l, loc=<0 0 0>, rot=<0 0 0 1>}
 deform_upperarm_fr_base_l {parent=upperarm_twist_01_l, loc=<-5.6312 5.092 0.3354>, rot=<0 0 0 -1>}
 deform_upperarm_fr_l {parent=deform_upperarm_fr_base_l, loc=<0 0 0>, rot=<0 0 0 1>}
 deform_upperarm_up_base_l {parent=upperarm_twist_01_l, loc=<-8.891 0.4825 3.9517>, rot=<0 0 0 -1>}
 deform_upperarm_up_l {parent=deform_upperarm_up_base_l, loc=<0 0 0>, rot=<0 0 0 1>}
 upperarm_twist_02_l {parent=upperarm_l, loc=<19.7998 -0.1134 0>, rot=<0 0 0 1>}
 clavicle_r {parent=spine_05, loc=<5.4831 -0.1053 3.2946>, rot=<0.7289 0.0465 -0.6829 0.009>}
 upperarm_r {parent=clavicle_r, loc=<-9.8658 1.2136 0.4105>, rot=<0.0164 -0.3245 -0.0269 -0.9454>}
 lowerarm_r {parent=upperarm_r, loc=<-26.465 0.8345 -0.259>, rot=<0.0036 0.0161 -0.2298 -0.9731>}
 hand_r {parent=lowerarm_r, loc=<-22.7355 0.3203 0.6106>, rot=<-0.6784 -0.0223 0.0735 -0.7307>}
 thumb_01_r {parent=hand_r, loc=<-1.3118 0.3191 2.459>, rot=<0.6546 -0.4042 0.0664 -0.6354>}
 thumb_02_r {parent=thumb_01_r, loc=<-5.9336 0.0448 -0.0401>, rot=<-0.0037 0.0213 0.298 -0.9543>}
 thumb_03_r {parent=thumb_02_r, loc=<-2.2223 -0.3682 0.0907>, rot=<0.0118 -0.013 0.0066 -0.9998>}
 index_metacarpal_r {parent=hand_r, loc=<-1.6535 -0.0183 2.96>, rot=<0.0158 -0.0124 -0.0173 -0.9996>}
 index_01_r {parent=index_metacarpal_r, loc=<-8.3008 -0.0357 0.0524>, rot=<-0.0044 0.0485 0.0924 -0.9945>}
 index_02_r {parent=index_01_r, loc=<-3.3587 0.3004 -0.0452>, rot=<-0.002 0.0027 0.3225 -0.9466>}
 index_03_r {parent=index_02_r, loc=<-2.5578 0.123 0.0252>, rot=<-0.002 0.0067 0.2558 -0.9667>}
 middle_metacarpal_r {parent=hand_r, loc=<-2.1856 0.1725 0.5984>, rot=<0.006 0.0168 -0.0329 -0.9993>}
 middle_01_r {parent=middle_metacarpal_r, loc=<-7.3275 0.2447 -0.0129>, rot=<-0.0208 0.0042 0.0858 -0.9961>}
 middle_02_r {parent=middle_01_r, loc=<-4.0123 -0.1401 -0.1175>, rot=<-0.0218 0.0258 0.3512 -0.9357>}
 middle_03_r {parent=middle_02_r, loc=<-2.6077 -0.0534 -0.1373>, rot=<0.0082 0.0188 0.1708 -0.9851>}
 ring_metacarpal_r {parent=hand_r, loc=<-2.2074 0.3625 -1.3638>, rot=<-0.0284 0.0722 -0.0458 -0.9959>}
 ring_01_r {parent=ring_metacarpal_r, loc=<-7.0295 0.6197 0.1446>, rot=<-0.0489 -0.0252 0.1892 -0.9804>}
 ring_02_r {parent=ring_01_r, loc=<-3.6174 -0.4647 0.1027>, rot=<0 0 0.2806 -0.9598>}
 ring_03_r {parent=ring_02_r, loc=<-2.4152 -0.114 -0.0199>, rot=<0 0 0.1189 -0.9929>}
 pinky_metacarpal_r {parent=hand_r, loc=<-1.7616 0.9753 -3.3814>, rot=<-0.0397 0.0814 -0.069 -0.9935>}
 pinky_01_r {parent=pinky_metacarpal_r, loc=<-7.1709 0.4831 0.1094>, rot=<-0.0654 -0.0035 0.2242 -0.9723>}
 pinky_02_r {parent=pinky_01_r, loc=<-2.8274 -0.0893 0.1043>, rot=<0.0037 0.0011 0.2854 -0.9584>}
 pinky_03_r {parent=pinky_02_r, loc=<-2.2046 -0.0307 0.0078>, rot=<-0.0002 0.0098 0.1445 -0.9895>}
 weapon_r {parent=hand_r, loc=<0.1974 3.4111 -0.3811>, rot=<0 0 0 1>}
 hand_attach_r {parent=hand_r, loc=<-8.5001 2.9999 0>, rot=<0 0 -0.7071 -0.7071>}
 deform_wrist_dn_base_r {parent=hand_r, loc=<-0.0001 1.5328 0>, rot=<0 0 0 -1>}
 deform_wrist_dn_r {parent=deform_wrist_dn_base_r, loc=<0 0 0>, rot=<0 0 0 1>}
 deform_wrist_up_base_r {parent=hand_r, loc=<-0.0002 -2.3007 -0.0001>, rot=<0 0 0 -1>}
 deform_wrist_up_r {parent=deform_wrist_up_base_r, loc=<0 0 0>, rot=<0 0 0 1>}
 lowerarm_twist_02_r {parent=lowerarm_r, loc=<-10.6707 0 0>, rot=<0 0 0 -1>}
 lowerarm_twist_01_r {parent=lowerarm_r, loc=<-18.316 0 0>, rot=<0 0 0 -1>}
 deform_elbow_in_base_r {parent=lowerarm_r, loc=<-0.1834 -3.6062 0.2572>, rot=<0 0 0 -1>}
 deform_elbow_in_r {parent=deform_elbow_in_base_r, loc=<0 0 0>, rot=<0 0 0 1>}
 deform_elbow_out_base_r {parent=lowerarm_r, loc=<0 4 -0.0001>, rot=<0 0 0 -1>}
 deform_elbow_out_r {parent=deform_elbow_out_base_r, loc=<0 0 0>, rot=<0 0 0 1>}
 upperarm_twist_01_r {parent=upperarm_r, loc=<-7.2135 0.1134 0>, rot=<0 0 0 -1>}
 deform_upperarm_bk_base_r {parent=upperarm_twist_01_r, loc=<6.4053 4.3746 -0.314>, rot=<0 0 0 -1>}
 deform_upperarm_bk_r {parent=deform_upperarm_bk_base_r, loc=<0 0 0>, rot=<0 0 0 1>}
 deform_upperarm_fr_base_r {parent=upperarm_twist_01_r, loc=<5.6312 -5.092 -0.3354>, rot=<0 0 0 -1>}
 deform_upperarm_fr_r {parent=deform_upperarm_fr_base_r, loc=<0 0 0>, rot=<0 0 0 1>}
 deform_upperarm_up_base_r {parent=upperarm_twist_01_r, loc=<8.891 -0.4825 -3.9517>, rot=<0 0 0 -1>}
 deform_upperarm_up_r {parent=deform_upperarm_up_base_r, loc=<0 0 0>, rot=<0 0 0 1>}
 upperarm_twist_02_r {parent=upperarm_r, loc=<-19.7998 0.1134 0>, rot=<0 0 0 -1>}
 dyn_shoelace_l_base {parent=spine_05, loc=<7.6337 4.9127 -4.5171>, rot=<-0.3248 -0.9298 -0.0461 -0.1672>}
 dyn_shoelace_l_01 {parent=dyn_shoelace_l_base, loc=<0 0 0>, rot=<0 0 0 1>}
 dyn_shoelace_l_02 {parent=dyn_shoelace_l_01, loc=<7.4929 0 0>, rot=<0 0 0 1>}
 dyn_shoelace_r_base {parent=spine_05, loc=<7.5315 5.309 4.5423>, rot=<-0.277 -0.9352 0.0208 0.2197>}
 dyn_shoelace_r_01 {parent=dyn_shoelace_r_base, loc=<0 0 0>, rot=<0 0 0 1>}
 dyn_shoelace_r_02 {parent=dyn_shoelace_r_01, loc=<6.7812 0 0>, rot=<0 0 0 1>}
 dyn_altskin_corr_ear_r_06 {parent=spine_05, loc=<13.988 -25.5564 22.2896>, rot=<0.4299 -0.38 0.765 -0.2926>}
 dyn_altskin_corr_ear_l_06 {parent=spine_05, loc=<13.9881 -25.5563 -22.2896>, rot=<0.2926 -0.765 0.38 -0.4299>}
 attach_backpack {parent=spine_05, loc=<-0.123 -13.473 0>, rot=<-0.0093 -0.707 0.0093 -0.707>}
 attach_cape {parent=spine_05, loc=<9.017 -11.714 0>, rot=<0.0616 -0.7044 -0.0616 -0.7044>}
 dyn_hood_base {parent=spine_05, loc=<0.4393 -9.2553 -0.019>, rot=<0.0401 -0.9992 0 0>}
 dyn_hood {parent=dyn_hood_base, loc=<0 0 0>, rot=<0 0 0 1>}
 dyn_hood_01 {parent=dyn_hood, loc=<0 0 0>, rot=<0 0 0 1>}
 dyn_ear_r_01 {parent=dyn_hood_01, loc=<0.4255 -3.0412 -5.9778>, rot=<0.1259 0.0107 0.9903 0.0579>}
 dyn_ear_r_02 {parent=dyn_ear_r_01, loc=<-10.9996 0.0001 -0.0001>, rot=<0 0 0 1>}
 dyn_ear_r_03 {parent=dyn_ear_r_02, loc=<-10.9996 0 -0.0001>, rot=<0 0 0.0019 -1>}
 dyn_ear_r_04 {parent=dyn_ear_r_03, loc=<-11.0003 0 0.0001>, rot=<0 0 0.0708 -0.9975>}
 dyn_ear_r_05 {parent=dyn_ear_r_04, loc=<-11 0 0>, rot=<0 0 0 1>}
 dyn_ear_l_base {parent=dyn_hood, loc=<-0.1023 -2.5699 5.4693>, rot=<-0.0091 0.1535 0.0582 -0.9864>}
 dyn_ear_l_01 {parent=dyn_ear_l_base, loc=<0 0 0>, rot=<0 -0.0279 0 -0.9996>}
 dyn_ear_l_02 {parent=dyn_ear_l_01, loc=<11 0 0>, rot=<0 0 0 1>}
 dyn_ear_l_03 {parent=dyn_ear_l_02, loc=<11 0 0>, rot=<0 0 0.0019 -1>}
 dyn_ear_l_04 {parent=dyn_ear_l_03, loc=<11 0 0>, rot=<0 0 0.0708 -0.9975>}
 dyn_ear_l_05 {parent=dyn_ear_l_04, loc=<11 0 0>, rot=<0 0 0 1>}
 deform_pec_aim_base_l {parent=spine_04, loc=<3.2456 9.3309 -1.2237>, rot=<0.0603 -0.4033 0.3318 -0.8507>}
 deform_pec_aim_l {parent=deform_pec_aim_base_l, loc=<0 0 0>, rot=<0 0 0 1>}
 deform_pec_base_l {parent=deform_pec_aim_l, loc=<9.3665 0 0>, rot=<0 0 0 1>}
 deform_pec_l {parent=deform_pec_base_l, loc=<0 0 0>, rot=<0 0 -0.6205 -0.7842>}
 deform_pec_aim_base_r {parent=spine_04, loc=<3.2452 9.3309 1.2237>, rot=<-0.4033 -0.0603 0.8507 0.3318>}
 deform_pec_aim_r {parent=deform_pec_aim_base_r, loc=<0 0 0>, rot=<0 0 0 1>}
 deform_pec_base_r {parent=deform_pec_aim_r, loc=<-9.3667 -0.0003 -0.0004>, rot=<0 0 0 1>}
 deform_pec_r {parent=deform_pec_base_r, loc=<0 0 0>, rot=<0 0 -0.6205 -0.7842>}
 thigh_l {parent=pelvis, loc=<-1.0855 -0.2109 -8.7053>, rot=<0.0009 0.0355 0.0155 -0.9993>}
 calf_l {parent=thigh_l, loc=<-42.5284 -0.0019 -0.7297>, rot=<0.0006 -0.0161 -0.0391 -0.9991>}
 foot_l {parent=calf_l, loc=<-46.7401 -2.5014 -0.592>, rot=<0.0427 -0.0238 0.0353 -0.9982>}
 ball_l {parent=foot_l, loc=<-3.5658 10.8066 -0.1544>, rot=<0.0001 -0.0001 -0.7063 -0.7079>}
 calf_twist_01_l {parent=calf_l, loc=<-35.8483 -1.9197 -0.5794>, rot=<0.0237 0.0095 -0.0035 -0.9997>}
 deform_knee_in_base_l {parent=calf_l, loc=<0.8031 -5.0231 0.2456>, rot=<0 0 0 -1>}
 deform_knee_in_l {parent=deform_knee_in_base_l, loc=<0 0 0>, rot=<0 0 0 1>}
 deform_knee_out_base_l {parent=calf_l, loc=<0 4 0>, rot=<0 0 0 -1>}
 deform_knee_out_l {parent=deform_knee_out_base_l, loc=<0 0 0>, rot=<0 0 0 1>}
 foot_l_1 {parent=calf_l, loc=<-39.6979 -0.5848 -0.3536>, rot=<0.0427 -0.0238 0.0353 -0.9982>}
 support_foot_l {parent=foot_l_1, loc=<-0.8071 -4.4746 -0.6304>, rot=<0 0 0 1>}
 ball_l_1 {parent=foot_l_1, loc=<-12.866 10.9537 -0.278>, rot=<0.0001 -0.0001 -0.7063 -0.7079>}
 calf_twist_02_l {parent=calf_l, loc=<-19.1546 -1.7943 -0.2638>, rot=<0.0237 0.0095 -0.0035 -0.9997>}
 thigh_twist_01_l {parent=thigh_l, loc=<-27.952 0 0>, rot=<0.0013 0 0 -1>}
 deform_groin_base_l {parent=thigh_l, loc=<0.4801 5.6776 -0.2486>, rot=<0 0 0 -1>}
 deform_groin_l {parent=deform_groin_base_l, loc=<0 0 0>, rot=<0 0 0 1>}
 thigh_r {parent=pelvis, loc=<-1.0855 -0.2109 8.7053>, rot=<0.0355 -0.0009 0.9993 0.0155>}
 calf_r {parent=thigh_r, loc=<42.5284 0.0019 0.7297>, rot=<0.0006 -0.0161 -0.0391 -0.9991>}
 foot_r {parent=calf_r, loc=<46.7401 2.5014 0.592>, rot=<0.0427 -0.0238 0.0353 -0.9982>}
 ball_r {parent=foot_r, loc=<3.5658 -10.8066 0.1544>, rot=<0.0001 -0.0001 -0.7063 -0.7079>}
 calf_twist_02_r {parent=calf_r, loc=<19.1546 1.7943 0.2638>, rot=<0.0237 0.0095 -0.0035 -0.9997>}
 calf_twist_01_r {parent=calf_r, loc=<35.8483 1.9197 0.5794>, rot=<0.0237 0.0095 -0.0035 -0.9997>}
 deform_knee_in_base_r {parent=calf_r, loc=<-0.8031 5.0231 -0.2457>, rot=<0 0 0 -1>}
 deform_knee_in_r {parent=deform_knee_in_base_r, loc=<0 0 0>, rot=<0 0 0 1>}
 deform_knee_out_base_r {parent=calf_r, loc=<0 -4 0>, rot=<0 0 0 -1>}
 deform_knee_out_r {parent=deform_knee_out_base_r, loc=<0 0 0>, rot=<0 0 0 1>}
 foot_r_1 {parent=calf_r, loc=<39.6979 0.5848 0.3536>, rot=<0.0427 -0.0238 0.0353 -0.9982>}
 support_foot_l1 {parent=foot_r_1, loc=<0.807 4.4746 0.6304>, rot=<0 0 0 1>}
 ball_r_1 {parent=foot_r_1, loc=<12.866 -10.9537 0.278>, rot=<0.0001 -0.0001 -0.7063 -0.7079>}
 thigh_twist_01_r {parent=thigh_r, loc=<27.952 0 0>, rot=<0.0013 0 0 -1>}
 deform_groin_base_r {parent=thigh_r, loc=<-0.4801 -5.6776 0.2486>, rot=<0 0 0 -1>}
 deform_groin_r {parent=deform_groin_base_r, loc=<0 0 0>, rot=<0 0 0 1>}
 deform_glute_base_l {parent=pelvis, loc=<-7.3466 -9.0213 -9.1245>, rot=<0.0172 0.031 -0.4477 -0.8935>}
 deform_glute_l {parent=deform_glute_base_l, loc=<0 0 0>, rot=<0 0 0 1>}
 deform_glute_base_r {parent=pelvis, loc=<-7.3464 -9.0213 9.1245>, rot=<0.031 -0.0172 0.8935 -0.4477>}
 deform_glute_r {parent=deform_glute_base_r, loc=<0 0 0>, rot=<0 0 0 1>}
 dyn_tail_base {parent=pelvis, loc=<7.358 -10.1209 0>, rot=<-0.0179 -0.9998 0 0>}
 dyn_tail_01 {parent=dyn_tail_base, loc=<0 0 0>, rot=<0 0 0 1>}
 dyn_bag_l_base {parent=pelvis, loc=<6.0501 -0.7401 -14.702>, rot=<-0.0171 -0.9538 0.0054 -0.3>}
 dyn_bag_l_01 {parent=dyn_bag_l_base, loc=<0 0 0>, rot=<0 0 0 1>}
 dyn_bag_r_base {parent=pelvis, loc=<5.6285 -0.7417 15.0348>, rot=<-0.0058 -0.3243 0.0169 -0.9458>}
 dyn_bag_r_01 {parent=dyn_bag_r_base, loc=<0 0 0>, rot=<0 0 0 1>}
 attach {parent=root, loc=<0 0 0>, rot=<0 0 0 1>}
 ik_hand_root {parent=root, loc=<0 0 0>, rot=<0 0 0 1>}
 ik_hand_gun {parent=ik_hand_root, loc=<-48.1743 3.3545 104.4381>, rot=<0.6276 0.1309 0.4183 -0.6434>}
 ik_hand_r {parent=ik_hand_gun, loc=<0 0 0>, rot=<0 0 0 1>}
 ik_hand_l {parent=ik_hand_gun, loc=<59.3271 67.6997 34.3524>, rot=<-0.6158 -0.7027 -0.3565 0>}
 ik_foot_root {parent=root, loc=<0 0 0>, rot=<0 0 0 1>}
 ik_foot_l {parent=ik_foot_root, loc=<14.8262 -7.3545 8.296>, rot=<0.0357 0.7045 0.0284 -0.7082>}
 ik_foot_r {parent=ik_foot_root, loc=<-14.8262 -7.3545 8.296>, rot=<0.7082 0.0284 -0.7045 0.0357>}
 root_fp {parent=root, loc=<0 0 0>, rot=<0 0 0 1>}
 attach_fp {parent=root_fp, loc=<0 0 0>, rot=<0 0 0 1>}
 camera_root {parent=root_fp, loc=<0 0 0>, rot=<0 0 0 1>}
 camera {parent=camera_root, loc=<0 0 0>, rot=<0 0 0 1>}
 attach_emote_01 {parent=root, loc=<0 0 0>, rot=<0 0 0 1>}
 attach_emote_02 {parent=root, loc=<0 0 0>, rot=<0 0 0 1>}
 attach_emote_03 {parent=root, loc=<0 0 0>, rot=<0 0 0 1>}
 warp_point_02 {parent=root, loc=<0 0 0>, rot=<0 0 0 1>}
 warp_point_03 {parent=root, loc=<0 0 0>, rot=<0 0 0 1>}
 warp_point_01 {parent=root, loc=<0 0 0>, rot=<0 0 0 1>}
]
variables [Backpack cape {type=bool, default=False}]
