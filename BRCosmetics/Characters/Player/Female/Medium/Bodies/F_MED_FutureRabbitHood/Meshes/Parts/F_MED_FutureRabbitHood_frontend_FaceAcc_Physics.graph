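# Physics asset for the frontend face-accessory mesh of the female medium FutureRabbitHood body, setting up secondary motion for its rabbit ears and lace strands.
# Capsules and spheres on pelvis, spine, neck, head, clavicles and upper arms form the body collision.
# Five-box ear chains hang from locked, zero-size base bodies on the head.
# Two-capsule "shoelace" chains hang from locked bases on spine_05, all chains using tight swing and twist limits.
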PhysicsAsset /BRCosmetics/Characters/Player/Female/Medium/Bodies/F_MED_FutureRabbitHood/Meshes/Parts/F_MED_FutureRabbitHood_frontend_FaceAcc_Physics
bodies {
 clavicle_l [{shapeType=Capsule, length=14, radius=5, shapeName=clavicle_l_capsule}]
 clavicle_r [{shapeType=Capsule, length=14, radius=5, shapeName=clavicle_r_capsule}]
 dyn_hood_ear_l_01 [{
   shapeType=Box
   extentX=10.9333
   extentY=8.9398
   extentZ=5.9654
   shapeName=dyn_hood_ear_l_01_box
   center=<3.5749 -2.1617 1.8522>
   rotation=<rot -0.0017 -0.647 19.7902>
  }]
 dyn_hood_ear_l_02 [{
   shapeType=Box
   extentX=10.6545
   extentY=9.621
   extentZ=4.6228
   shapeName=dyn_hood_ear_l_02_box
   center=<1.8293 -0.8949 0.0257>
   rotation=<rot -0.0001 0 -18.5664>
  }]
 dyn_hood_ear_l_03 [{
   shapeType=Box
   extentX=13.3534
   extentY=11.9612
   extentZ=5.066
   shapeName=dyn_hood_ear_l_03_box
   center=<3.132 -1.1705 -0.7633>
   rotation=<rot 2.7499 4.0433 -34.1669>
  }]
 dyn_hood_ear_l_04 [{
   shapeType=Box
   extentX=10.2954
   extentY=12.1897
   extentZ=4.5022
   shapeName=dyn_hood_ear_l_04_box
   center=<4.4426 -0.5097 -0.3312>
   rotation=<rot -0.0003 0.0001 -33.0123>
  }]
 dyn_hood_ear_l_05 [{
   shapeType=Box
   extentX=15.711
   extentY=11.9031
   extentZ=3.9766
   shapeName=dyn_hood_ear_l_05_box
   center=<5.9837 -0.4085 -0.2359>
   rotation=<rot -0.0005 0 -30.0001>
  }]
 dyn_hood_ear_l_base [{shapeType=Box, extentX=0, extentY=0, extentZ=0, shapeName=dyn_hood_ear_l_base_box}]
 dyn_hood_ear_r_01 [{
   shapeType=Box
   extentX=11.6968
   extentY=10.1396
   extentZ=6.9874
   shapeName=dyn_hood_ear_r_01_box
   center=<2.5835 0.3504 2.1226>
   rotation=<rot 0 0 -35.4752>
  }]
 dyn_hood_ear_r_02 [{
   shapeType=Box
   extentX=12.607
   extentY=8.6169
   extentZ=4.5566
   shapeName=dyn_hood_ear_r_02_box
   center=<6.39 1.3094 0.1122>
   rotation=<rot -0.0001 0 19.9515>
  }]
 dyn_hood_ear_r_03 [{
   shapeType=Box
   extentX=11.902
   extentY=10.0118
   extentZ=4.4667
   shapeName=dyn_hood_ear_r_03_box
   center=<5.4604 1.0438 -0.8331>
   rotation=<rot -0.0001 -0.0001 33.5892>
  }]
 dyn_hood_ear_r_04 [{
   shapeType=Box
   extentX=12.1668
   extentY=11.1935
   extentZ=3.772
   shapeName=dyn_hood_ear_r_04_box
   center=<4.9736 0.4718 -0.1383>
   rotation=<rot -0.0001 0 37.7002>
  }]
 dyn_hood_ear_r_05 [{
   shapeType=Box
   extentX=14.8455
   extentY=9.2917
   extentZ=3.5766
   shapeName=dyn_hood_ear_r_05_box
   center=<6.1473 0.2827 0.3369>
   rotation=<rot -0.0002 -0.0001 40.0001>
  }]
 dyn_hood_ear_r_base [{shapeType=Box, extentX=0, extentY=0, extentZ=0, shapeName=dyn_hood_ear_r_base_box}]
 dyn_shoelace_l_01 [{
   shapeType=Capsule
   length=3.9204
   radius=1.1554
   shapeName=dyn_shoelace_l_01_capsule
   center=<3.7127 0.3035 0.4147>
   rotation=<rot 0 84.9018 -87.8172>
  }]
 dyn_shoelace_l_02 [{
   shapeType=Capsule
   length=6.5477
   radius=1.1256
   shapeName=dyn_shoelace_l_02_capsule
   center=<1.1301 -0.1748 0.4125>
   rotation=<rot 0 -101.8397 -91.5885>
  }]
 dyn_shoelace_l_base [{shapeType=Capsule, length=0.505, radius=0.505, shapeName=dyn_shoelace_l_base_capsule, rotation=<rot 0 -90 0>}]
 dyn_shoelace_r_01 [{
   shapeType=Capsule
   length=2.8
   radius=1.2407
   shapeName=dyn_shoelace_r_01_capsule
   center=<3.6011 0.4766 0.0247>
   rotation=<rot 0 92.8751 -88.7862>
  }]
 dyn_shoelace_r_02 [{
   shapeType=Capsule
   length=6.292
   radius=1.3262
   shapeName=dyn_shoelace_r_02_capsule
   center=<2.4995 0.0787 0.5439>
   rotation=<rot -0.6459 86.485 -82.0631>
  }]
 dyn_shoelace_r_base [{shapeType=Capsule, length=0.505, radius=0.505, shapeName=dyn_shoelace_r_base_capsule, rotation=<rot 0 -90 0>}]
 head [
  {shapeType=Sphere, radius=8.4415, shapeName=head_sphere, center=<5.3728 -3.0511 0>}
  {shapeType=Sphere, radius=10.1727, shapeName=head_sphere, center=<-3.4321 -3.8684 0>}
  {shapeType=Capsule, length=4.4306, radius=8.1191, shapeName=head_capsule, center=<2.3857 0 0>}
  {shapeType=Capsule, length=12.0938, radius=6.8321, shapeName=head_capsule, center=<-6.8943 -4.2734 0>}
 ]
 neck_01 [{shapeType=Capsule, length=4.5523, radius=10.3189, shapeName=neck_01_capsule}]
 neck_02 [{shapeType=Capsule, length=3.6627, radius=9.9783, shapeName=neck_02_capsule}]
 pelvis [{shapeType=Capsule, length=14, radius=5, shapeName=pelvis_capsule}]
 spine_01 [{shapeType=Capsule, length=13.9265, radius=7.6973, shapeName=spine_01_capsule}]
 spine_02 [{shapeType=Capsule, length=13.9265, radius=7.6973, shapeName=spine_02_capsule}]
 spine_03 [{shapeType=Capsule, length=13.9265, radius=7.6973, shapeName=spine_03_capsule}]
 spine_04 [{shapeType=Capsule, length=12.4677, radius=10.1606, shapeName=spine_04_capsule, center=<0 -0.6902 0>}]
 spine_05 [
  {shapeType=Capsule, length=21.1459, radius=6.9048, shapeName=spine_05_capsule, center=<-3.4639 3.7949 0>}
  {shapeType=Capsule, length=22.4056, radius=7.5944, shapeName=spine_05_capsule}
 ]
 upperarm_l [{shapeType=Capsule, length=26.5606, radius=5, shapeName=upperarm_l_capsule, center=<14.0374 0 0.048>, rotation=<rot -89.8041 0.0519 -0.0518>}]
 upperarm_r [{
   shapeType=Capsule
   length=26.5606
   radius=5
   shapeName=upperarm_r_capsule
   center=<-14.0374 0 -0.048>
   rotation=<rot -89.8041 0.0519 179.9482>
  }]
}
constraints [
 {bone1Name=dyn_shoelace_l_01, bone2Name=dyn_shoelace_l_base, swing1LimitDegrees=51.92, swing2LimitDegrees=11, twistLimitDegrees=8}
 {bone1Name=dyn_shoelace_l_02, bone2Name=dyn_shoelace_l_01, swing1LimitDegrees=30.32, swing2LimitDegrees=18.2, twistLimitDegrees=8}
 {bone1Name=dyn_shoelace_r_01, bone2Name=dyn_shoelace_r_base, swing1LimitDegrees=51.92, swing2LimitDegrees=11, twistLimitDegrees=8}
 {bone1Name=dyn_shoelace_r_02, bone2Name=dyn_shoelace_r_01, swing1LimitDegrees=26, swing2LimitDegrees=16.76, twistLimitDegrees=8}
 {
  bone1Name=dyn_shoelace_l_base
  bone2Name=spine_05
  swing1LimitDegrees=45
  swing1Motion=Locked
  swing2LimitDegrees=45
  swing2Motion=Locked
  twistLimitDegrees=45
  twistMotion=Locked
 }
 {
  bone1Name=dyn_shoelace_r_base
  bone2Name=spine_05
  swing1LimitDegrees=45
  swing1Motion=Locked
  swing2LimitDegrees=45
  swing2Motion=Locked
  twistLimitDegrees=45
  twistMotion=Locked
 }
 {
  bone1Name=spine_01
  bone2Name=pelvis
  swing1LimitDegrees=45
  swing1Motion=Free
  swing2LimitDegrees=45
  swing2Motion=Free
  twistLimitDegrees=45
  twistMotion=Free
 }
 {
  bone1Name=spine_02
  bone2Name=spine_01
  swing1LimitDegrees=45
  swing1Motion=Free
  swing2LimitDegrees=45
  swing2Motion=Free
  twistLimitDegrees=45
  twistMotion=Free
 }
 {
  bone1Name=spine_03
  bone2Name=spine_02
  swing1LimitDegrees=45
  swing1Motion=Free
  swing2LimitDegrees=45
  swing2Motion=Free
  twistLimitDegrees=45
  twistMotion=Free
 }
 {
  bone1Name=spine_04
  bone2Name=spine_03
  swing1LimitDegrees=45
  swing1Motion=Locked
  swing2LimitDegrees=45
  swing2Motion=Locked
  twistLimitDegrees=45
  twistMotion=Locked
 }
 {
  bone1Name=spine_05
  bone2Name=spine_04
  swing1LimitDegrees=45
  swing1Motion=Locked
  swing2LimitDegrees=45
  swing2Motion=Locked
  twistLimitDegrees=45
  twistMotion=Locked
 }
 {
  bone1Name=clavicle_l
  bone2Name=spine_05
  swing1LimitDegrees=45
  swing1Motion=Locked
  swing2LimitDegrees=45
  swing2Motion=Locked
  twistLimitDegrees=45
  twistMotion=Locked
 }
 {
  bone1Name=upperarm_l
  bone2Name=clavicle_l
  swing1LimitDegrees=45
  swing1Motion=Free
  swing2LimitDegrees=45
  swing2Motion=Free
  twistLimitDegrees=45
  twistMotion=Free
 }
 {
  bone1Name=clavicle_r
  bone2Name=spine_05
  swing1LimitDegrees=45
  swing1Motion=Locked
  swing2LimitDegrees=45
  swing2Motion=Locked
  twistLimitDegrees=45
  twistMotion=Locked
 }
 {
  bone1Name=upperarm_r
  bone2Name=clavicle_r
  swing1LimitDegrees=45
  swing1Motion=Free
  swing2LimitDegrees=45
  swing2Motion=Free
  twistLimitDegrees=45
  twistMotion=Free
 }
 {
  bone1Name=neck_01
  bone2Name=spine_05
  swing1LimitDegrees=45
  swing1Motion=Locked
  swing2LimitDegrees=45
  swing2Motion=Locked
  twistLimitDegrees=45
  twistMotion=Locked
 }
 {
  bone1Name=neck_02
  bone2Name=neck_01
  swing1LimitDegrees=45
  swing1Motion=Locked
  swing2LimitDegrees=45
  swing2Motion=Locked
  twistLimitDegrees=45
  twistMotion=Locked
 }
 {
  bone1Name=head
  bone2Name=neck_02
  swing1LimitDegrees=45
  swing1Motion=Locked
  swing2LimitDegrees=45
  swing2Motion=Locked
  twistLimitDegrees=45
  twistMotion=Locked
 }
 {
  bone1Name=dyn_hood_ear_l_base
  bone2Name=head
  swing1LimitDegrees=45
  swing1Motion=Locked
  swing2LimitDegrees=45
  swing2Motion=Locked
  twistLimitDegrees=45
  twistMotion=Locked
 }
 {bone1Name=dyn_hood_ear_l_01, bone2Name=dyn_hood_ear_l_base, swing1LimitDegrees=11, swing2LimitDegrees=14, twistLimitDegrees=7}
 {bone1Name=dyn_hood_ear_l_02, bone2Name=dyn_hood_ear_l_01, swing1LimitDegrees=11, swing2LimitDegrees=18.32, twistLimitDegrees=7}
 {bone1Name=dyn_hood_ear_l_03, bone2Name=dyn_hood_ear_l_02, swing1LimitDegrees=11, swing2LimitDegrees=18.32, twistLimitDegrees=7}
 {bone1Name=dyn_hood_ear_l_04, bone2Name=dyn_hood_ear_l_03, swing1LimitDegrees=11, swing2LimitDegrees=14, twistLimitDegrees=7}
 {bone1Name=dyn_hood_ear_l_05, bone2Name=dyn_hood_ear_l_04, swing1LimitDegrees=11, swing2LimitDegrees=14, twistLimitDegrees=7}
 {
  bone1Name=dyn_hood_ear_r_base
  bone2Name=head
  swing1LimitDegrees=45
  swing1Motion=Locked
  swing2LimitDegrees=45
  swing2Motion=Locked
  twistLimitDegrees=45
  twistMotion=Locked
 }
 {bone1Name=dyn_hood_ear_r_01, bone2Name=dyn_hood_ear_r_base, swing1LimitDegrees=11, swing2LimitDegrees=15, twistLimitDegrees=5}
 {bone1Name=dyn_hood_ear_r_02, bone2Name=dyn_hood_ear_r_01, swing1LimitDegrees=11, swing2LimitDegrees=17, twistLimitDegrees=7}
 {bone1Name=dyn_hood_ear_r_03, bone2Name=dyn_hood_ear_r_02, swing1LimitDegrees=11, swing2LimitDegrees=17, twistLimitDegrees=7}
 {bone1Name=dyn_hood_ear_r_04, bone2Name=dyn_hood_ear_r_03, swing1LimitDegrees=11, swing2LimitDegrees=17, twistLimitDegrees=7}
 {bone1Name=dyn_hood_ear_r_05, bone2Name=dyn_hood_ear_r_04, swing1LimitDegrees=11, swing2LimitDegrees=17, twistLimitDegrees=7}
]
