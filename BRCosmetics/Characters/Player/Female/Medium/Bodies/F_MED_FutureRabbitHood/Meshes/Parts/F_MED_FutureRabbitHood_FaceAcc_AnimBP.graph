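# Post-process-style part AnimBP for the FutureRabbitHood FaceAcc cosmetic, adding secondary motion to its hair and hood bones.
# The pose copied from the body runs through a Control Rig, then one of four RigidBody physics assets chosen by front-end and pack state.
# AnimDynamics springs on hair bones and an additive hood-bone offset finish the pose.
# Events set gravity from a parameter asset and derive pack state from backbling and rigid-body checks.
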AnimBlueprint /BRCosmetics/Characters/Player/Female/Medium/Bodies/F_MED_FutureRabbitHood/Meshes/Parts/F_MED_FutureRabbitHood_FaceAcc_AnimBP
target_skeleton=&/F_MED_FutureRabbitHood_FaceAcc_Skeleton
graphs [
 AnimGraph {nodes [
   n1 = RigidBody!(
    ComponentPose=Local To Component(LocalPose=Use cached pose 'main cache'())
    Alpha
    OverrideWorldGravity:Vector=Get Gravity Override()
    ExternalForce
   ) -> (Pose) {
    bClampLinearTranslationLimitToRefPose=true
    bForceDisableCollisionBetweenConstraintBodies=true
    bOverrideWorldGravity=true
    overridePhysicsAsset=&/F_MED_FutureRabbitHood_frontend_FaceAcc_Physics
    simulationSpace=BaseBoneSpace
    alphaBoolBlend {blendOption=Linear}
    baseBoneRef {boneName=spine_05}
    simSpaceSettings {
     dampingAlpha=0
     maxAngularAcceleration=40
     maxAngularVelocity=6
     maxLinearAcceleration=1500
     maxLinearVelocity=700
     velocityScaleZ=0.7
     worldAlpha=1
     externalLinearDragV {x=2}
    }
   }
   n2 = RigidBody(ComponentPose=Local To Component(LocalPose=Use cached pose 'main cache'()), OverrideWorldGravity=Get Gravity Override()) {
    bClampLinearTranslationLimitToRefPose=true
    bForceDisableCollisionBetweenConstraintBodies=true
    bOverrideWorldGravity=true
    overridePhysicsAsset=&/F_MED_FutureRabbitHood_pack_fend_FaceAcc_Physics
    simulationSpace=BaseBoneSpace
    alphaBoolBlend {blendOption=Linear}
    baseBoneRef {boneName=spine_05}
    simSpaceSettings {
     dampingAlpha=0
     maxAngularAcceleration=40
     maxAngularVelocity=6
     maxLinearAcceleration=1500
     maxLinearVelocity=700
     velocityScaleZ=0.7
     worldAlpha=1
     externalLinearDragV {x=2}
    }
   }
   n3 = Blend Poses by bool(
    BlendPose_0=Component To Local(ComponentPose=@n2)
    BlendPose_1=Component To Local(ComponentPose=@n1)
    bActiveValue=Get Is Wearing Pack() -> (Is Wearing Pack:bool)
   )
   n4 = RigidBody(ComponentPose=Local To Component(LocalPose=Use cached pose 'main cache'()), OverrideWorldGravity=Get Gravity Override()) {
    bClampLinearTranslationLimitToRefPose=true
    bForceDisableCollisionBetweenConstraintBodies=true
    bOverrideWorldGravity=true
    overridePhysicsAsset=&/F_MED_FutureRabbitHood_backpack_FaceAcc_Physics
    simulationSpace=BaseBoneSpace
    alphaBoolBlend {blendOption=Linear}
    baseBoneRef {boneName=spine_05}
    simSpaceSettings {
     dampingAlpha=0
     maxAngularAcceleration=40
     maxAngularVelocity=6
     maxLinearAcceleration=1500
     maxLinearVelocity=700
     velocityScaleZ=0.7
     worldAlpha=1
     externalLinearDragV {x=2}
    }
   }
   n5 = RigidBody(ComponentPose=Local To Component(LocalPose=Use cached pose 'main cache'()), OverrideWorldGravity=Get Gravity Override()) {
    bClampLinearTranslationLimitToRefPose=true
    bForceDisableCollisionBetweenConstraintBodies=true
    bOverrideWorldGravity=true
    overridePhysicsAsset=&/F_MED_FutureRabbitHood_FaceAcc_Physics
    simulationSpace=BaseBoneSpace
    alphaBoolBlend {blendOption=Linear}
    baseBoneRef {boneName=spine_05}
    simSpaceSettings {
     dampingAlpha=0
     maxAngularAcceleration=40
     maxAngularVelocity=6
     maxLinearAcceleration=1500
     maxLinearVelocity=700
     velocityScaleZ=0.7
     worldAlpha=1
     externalLinearDragV {x=2}
    }
   }
   n6 = Blend Poses by bool(BlendPose_0=Component To Local(ComponentPose=@n4), BlendPose_1=Component To Local(ComponentPose=@n5), bActiveValue=Get Is Wearing Pack())
   n7 = Get bIsInFrontEnd()
   n8 = Anim Dynamics - Bone: dyn_hair_l_01(
    ComponentPose=Local To Component(LocalPose=Blend Poses by bool(BlendPose_0=@n3, BlendPose_1=@n6, bActiveValue=@n7))
    Alpha=1
    ExternalForce="0,0,0"
   ) {
    angularDampingOverride=1
    angularSpringConstant=200
    bAngularSpring=true
    bOverrideAngularDamping=true
    bUseGravityOverride=true
    simulationSpace=RootRelative
    boundBone {boneName=dyn_hair_l_01}
    componentAppliedLinearAccClamp=<500 500 500>
    componentLinearAccScale=<0.1 0.1 0.3>
    componentLinearVelScale=<0.1 0.1 0.3>
    gravityOverride {z=-980}
    physicsBodyDefinitions [{
      boundBone {boneName=dyn_hair_l_01}
      boxExtents=<21 21 21>
      constraintSetup {angularLimitsMax=<18 6 33>, angularLimitsMin=<-20 1 -8>, angularTarget=<1 0 0>}
      localJointOffset=<1 0 0.3>
     }]
    relativeSpaceBone {boneName=head}
    simSpaceSettings {maxAngularAcceleration=35, maxAngularVelocity=7, simSpaceAngularAlpha=1}
   }
   n9 = Anim Dynamics - Bone: dyn_hair_l_02(ComponentPose=@n8, Alpha=1, ExternalForce="0,0,0") {
    angularDampingOverride=1
    angularSpringConstant=200
    bAngularSpring=true
    bOverrideAngularDamping=true
    bUseGravityOverride=true
    simulationSpace=RootRelative
    boundBone {boneName=dyn_hair_l_02}
    componentAppliedLinearAccClamp=<500 500 500>
    componentLinearAccScale=<0.1 0.1 0.3>
    componentLinearVelScale=<0.1 0.1 0.3>
    gravityOverride {z=-980}
    physicsBodyDefinitions [{
      boundBone {boneName=dyn_hair_l_02}
      boxExtents=<9 9 9>
      constraintSetup {angularLimitsMax=<18 4 15>, angularLimitsMin=<-12 -11 -11>, angularTarget=<1 0 0>}
      localJointOffset=<1 0 0.3>
     }]
    relativeSpaceBone {boneName=head}
    simSpaceSettings {maxAngularAcceleration=35, maxAngularVelocity=7, simSpaceAngularAlpha=1}
   }
   n10 = Anim Dynamics - Bone: dyn_hair_r_01(ComponentPose=@n9, Alpha=1, ExternalForce="0,0,0") {
    angularDampingOverride=1
    angularSpringConstant=200
    bAngularSpring=true
    bOverrideAngularDamping=true
    bUseGravityOverride=true
    simulationSpace=RootRelative
    boundBone {boneName=dyn_hair_r_01}
    componentAppliedLinearAccClamp=<500 500 500>
    componentLinearAccScale=<0.1 0.1 0.3>
    componentLinearVelScale=<0.1 0.1 0.3>
    gravityOverride {z=-980}
    physicsBodyDefinitions [{
      boundBone {boneName=dyn_hair_r_01}
      boxExtents=<12 12 12>
      constraintSetup {angularLimitsMax=<12 7 4>, angularLimitsMin=<-20 2 -33>, angularTarget=<1 0 0>}
      localJointOffset=<1 0 0.3>
     }]
    relativeSpaceBone {boneName=head}
    simSpaceSettings {maxAngularAcceleration=35, maxAngularVelocity=7, simSpaceAngularAlpha=1}
   }
   n11 = Anim Dynamics - Bone: dyn_hair_r_02(ComponentPose=@n10, Alpha=1, ExternalForce="0,0,0") {
    angularDampingOverride=1
    angularSpringConstant=200
    bAngularSpring=true
    bOverrideAngularDamping=true
    bUseGravityOverride=true
    simulationSpace=RootRelative
    boundBone {boneName=dyn_hair_r_02}
    componentAppliedLinearAccClamp=<500 500 500>
    componentLinearAccScale=<0.1 0.1 0.3>
    componentLinearVelScale=<0.1 0.1 0.3>
    gravityOverride {z=-980}
    physicsBodyDefinitions [{
      boundBone {boneName=dyn_hair_r_02}
      boxExtents=<12 12 12>
      constraintSetup {angularLimitsMax=<12 22 5>, angularLimitsMin=<-12 -11 -15>, angularTarget=<1 0 0>}
      localJointOffset=<1 0 0.3>
     }]
    relativeSpaceBone {boneName=head}
    simSpaceSettings {maxAngularAcceleration=35, maxAngularVelocity=7, simSpaceAngularAlpha=1}
   }
   n12 = Anim Dynamics - Bone: dyn_hair_front_1_l_base(ComponentPose=@n11, Alpha=1, ExternalForce="0,0,0") {
    angularDampingOverride=0.999
    angularSpringConstant=200
    bAngularSpring=true
    bOverrideAngularDamping=true
    bUseGravityOverride=true
    simulationSpace=RootRelative
    boundBone {boneName=dyn_hair_front_1_l_base}
    componentAppliedLinearAccClamp=<3000 3000 3000>
    componentLinearAccScale=<0.1 0.1 0.4>
    componentLinearVelScale=<0.1 0.1 0.4>
    gravityOverride {z=-980}
    physicsBodyDefinitions [{
      boundBone {boneName=dyn_hair_front_1_l_base}
      boxExtents=<12 12 12>
      constraintSetup {angularLimitsMax=<9 22 22>, angularLimitsMin=<-9 -22 -1>, angularTarget=<1 0 0>}
      localJointOffset=<1 1 0>
     }]
    relativeSpaceBone {boneName=spine_05}
    simSpaceSettings {maxAngularAcceleration=40, maxAngularVelocity=8, simSpaceAngularAlpha=0.5}
   }
   n13 = Anim Dynamics - Bone: dyn_hair_front_base(ComponentPose=@n12, Alpha=1, ExternalForce="0,0,0") {
    angularDampingOverride=0.999
    angularSpringConstant=200
    bAngularSpring=true
    bOverrideAngularDamping=true
    bUseGravityOverride=true
    simulationSpace=RootRelative
    boundBone {boneName=dyn_hair_front_base}
    componentAppliedLinearAccClamp=<3000 3000 3000>
    componentLinearAccScale=<0.1 0.1 0.4>
    componentLinearVelScale=<0.1 0.1 0.4>
    gravityOverride {z=-980}
    physicsBodyDefinitions [{
      boundBone {boneName=dyn_hair_front_base}
      boxExtents=<14 14 14>
      constraintSetup {angularLimitsMax=<9 22 22>, angularLimitsMin=<-9 -22 -1>, angularTarget=<1 0 0>}
      localJointOffset=<1 1 0>
     }]
    relativeSpaceBone {boneName=spine_05}
    simSpaceSettings {maxAngularAcceleration=40, maxAngularVelocity=8, simSpaceAngularAlpha=0.5}
   }
   Output Pose(Result=Component To Local(ComponentPose="Transform (Modify) Bone - Bone: corr_hood_1"(ComponentPose=@n13, Alpha=1, Translation="0,0,8", Rotation="0,0,0", Scale="1,1,1") {translationMode=BMM_Additive, boneToModify {boneName=corr_hood_1}, translation {z=8}})) {layerGroup=None}
   n14 = Control Rig!(
    Source=Copy Pose From Mesh(SourceMeshComponent=Get Body() -> (Body:Skeletal Mesh Component Object Reference)) {bUseAttachedParent=true}
    Alpha
    Backpack cape:bool=Get Is Wearing Pack()
   ) -> (Pose) {controlRigAssetReference {blueprintRigClass=&/F_MED_FutureRabbitHood_FaceAcc_CtrlRig.F_MED_FutureRabbitHood_FaceAcc_CtrlRig_C}}
   Save cached pose 'main cache'(Pose=@n14) {cacheName=main cache}
  ]}
 EventGraph {nodes [
   Event BlueprintUpdateAnimation()
   > Sequence()
   > n1 = "Gravity Override State Function (Parameter Asset) (Nativized)"(
    CharacterPartAnimInstance=Self-Reference()
    GameTimeInSeconds=GetGameTimeInSeconds()
    Parameters=/BRCosmetics/Characters/Player/Female/Medium/Bodies/F_MED_FutureRabbit/Meshes/F_MED_FutereRabbit_Parameters
   )
   > Set Gravity Override(Gravity Override=@n1)
   n2 = Event OnCharacterCustomizationLoaded()
   Is Valid(exec=[@n2.then, @n5.then], InputObject=@n2.CustomBaseMesh)
   > Set Body(Body:Skeletal Mesh Component Object Reference=@n2.CustomBaseMesh) -> (Output_Get:Skeletal Mesh Component Object Reference)
   n3 = Get bRigidBodyEnabled()
   n4 = CheckForBackbling()
   > Set Is Wearing Pack(Is Wearing Pack:bool=OR Boolean!(A=@n4, B=@n4.bIsBackblingACape, C:bool=@n4.bIsBackblingLong, D:bool=NOT Boolean(A=@n3)) -> (ReturnValue)) -> (Output_Get:bool)
   n5 = Event BlueprintInitializeAnimation()
  ]}
]
variables [
 Body {type=SkeletalMeshComponent}
 Gravity Override {type=Vector}
 Is Wearing Pack {type=boolean}
]
